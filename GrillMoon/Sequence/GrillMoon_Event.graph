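# GrillMoon_Event is a master Level Sequence for the GrillMoon event, with no bindings of its own.
# Sub-sequence tracks stage timer/masks, Mistmylar mask power and jade wanderer portal, then a kaiju sky fight and wandering demon death, alongside a full-length track.
# Two shockwave blasts overlap the fight's start and end.
# Light Beam and MM Meditates Data Layer tracks each split into two sections where the portal ends, just before the fight.
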LevelSequence /GrillMoon/Sequence/GrillMoon_Event
bindings []
display_rate {denominator=1, numerator=60}
tick_resolution {denominator=1, numerator=24000}
tracks [
 {type=MovieSceneSubTrack, display_name=wdd - wandering demon death, sections [{type=MovieSceneSubSection, end_frame=9775, start_frame=8999}]}
 {type=MovieSceneSubTrack, display_name=tdm - timer daily masks, sections [{type=MovieSceneSubSection, end_frame=601, start_frame=79}]}
 {type=MovieSceneSubTrack, display_name=fel - full event length, sections [{type=MovieSceneSubSection, end_frame=10500, start_frame=79}]}
 {type=MovieSceneSubTrack, display_name=mmp - mistmylar mask power, sections [{type=MovieSceneSubSection, end_frame=4103, start_frame=79}]}
 {type=MovieSceneSubTrack, display_name=pjw - portal jade wanderer, sections [{type=MovieSceneSubSection, end_frame=4106, start_frame=763}]}
 {type=MovieSceneSubTrack, display_name=swb - shockwave blast, sections [{type=MovieSceneSubSection, end_frame=4801, start_frame=3822}]}
 {type=MovieSceneSubTrack, display_name=swb - shockwave blast, sections [{type=MovieSceneSubSection, end_frame=10228, start_frame=9417}]}
 {type=MovieSceneSubTrack, display_name=kfs - kaiju fight sky, sections [{type=MovieSceneSubSection, end_frame=9819, start_frame=4108}]}
 {
  type=MovieSceneDataLayerTrack
  display_name=Data Layer - Light Beam
  sections [{type=MovieSceneDataLayerSection, end_frame=4106, start_frame=79}, {type=MovieSceneDataLayerSection, end_frame=10500, start_frame=4106}]
 }
 {
  type=MovieSceneDataLayerTrack
  display_name=Data Layer - MM Meditates
  sections [{type=MovieSceneDataLayerSection, end_frame=4106, start_frame=79}, {type=MovieSceneDataLayerSection, end_frame=10500, start_frame=4106}]
 }
]
